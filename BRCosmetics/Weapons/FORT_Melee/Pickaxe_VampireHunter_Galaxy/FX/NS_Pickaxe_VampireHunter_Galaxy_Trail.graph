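# Niagara weapon swing trail for the Galaxy Vampire Hunter pickaxe, built from three anim-trail ribbon emitters following bone/socket pairs on the skeletal mesh.
# GalaxyTrail drives dynamic material parameters with a per-emitter random value and a time-of-day-adjusted fade.
# NebulaTrail adds a bright-to-dark bluish color over life, while longer-lived SharpBladeTrail shifts cyan to purple and narrows.
# A shared system SwingGeneralFade scales all three trails' opacity or material fade.
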NiagaraSystem /BRCosmetics/Weapons/FORT_Melee/Pickaxe_VampireHunter_Galaxy/FX/NS_Pickaxe_VampireHunter_Galaxy_Trail
dependencies {
 usedDataInterfaces [&NiagaraDataInterfaceCurve, &NiagaraDataInterfaceColorCurve, &NiagaraDataInterfaceSkeletalMesh]
 usedDynamicInputs [&Frac_Float]
 usedModules [
  &self:SystemSpawnScript.SystemScriptSource.SystemScriptGraph.NiagaraNodeAssignment_1.SetVariables_#1
  &self:SystemSpawnScript.SystemScriptSource.SystemScriptGraph.NiagaraNodeAssignment_0.SetVariables_#2
  &SystemState
  &self:SystemSpawnScript.SystemScriptSource.SystemScriptGraph.NiagaraNodeAssignment_2.SetVariables_#3
 ]
}
emitters {
 GalaxyTrail {
  emitterUpdateScript [
   &EmitterState()
   &FN_SpawnAnimTrail(Spawn Interval=0)
   /SetVariables(Emitter.RandomFloat:NiagaraFloat=&RandomRangeFloat(Minimum=0, Maximum=1)) {isSetParameters=true, module=SetVariables_#4}
  ]
  particleSpawnScript [
   &FN_InitializeAnimTrailBoneIndecies(First Bone or Socket Index=1, Second Bone or Socket Index=0)
   &FN_InitializeAnimTrail(Lifetime=0.14, SkeletalMesh=$System.Skeletal Mesh) {module=FN_InitializeAnimTrail001}
   &AddVelocity(Velocity=<0 1000 0>, Velocity Speed Scale=2) {enabled=false}
  ]
  particleUpdateScript [
   &FN_UpdateAnimTrail(SkeletalMesh=$System.Skeletal Mesh) {module=FN_UpdateAnimTrail001}
   &SolveForcesAndVelocity(Clamp Velocity=false, Limit Acceleration=false)
   &DynamicMaterialParameters(
    Index 0 Param 1=$Emitter.RandomFloat
    Index 0 Param 4=&Multiply_Float(
     A=&Multiply_Float(
      A=$System.SwingGeneralFade
      B=&FN_Select_Float_from_Time_Of_Day(Morning=0.8, Day=0.73, Evening=0.8, Night=0.87)
     )
     B=&FloatFromCurve(FloatCurve=DataInterface() {propertyValues {Curve {keys [
          {interpMode=RCIM_Cubic, tangentMode=RCTM_User, time=0, value=1}
          {arriveTangent=-0.0076, interpMode=RCIM_Cubic, leaveTangent=-0.0076, tangentMode=RCTM_User, time=0.8, value=0}
         ]}}})
    )
    Module.Param0WriteEnabled=true
    Module.Param1WriteEnabled=false
    Module.Param2WriteEnabled=false
    Module.Param3WriteEnabled=true
   ) {module=DynamicMaterialParameters001}
  ]
  rendererClasses [&NiagaraRibbonRendererProperties]
  renderers [{rendererClass=&NiagaraRibbonRendererProperties}]
 }
 NebulaTrail {
  emitterUpdateScript [
   &EmitterState()
   &FN_SpawnAnimTrail(Spawn Interval=0)
   /SetVariables(Emitter.RandomFloat:NiagaraFloat=&RandomRangeFloat(Minimum=0, Maximum=1)) {isSetParameters=true, module=SetVariables_#4}
  ]
  particleSpawnScript [
   &FN_InitializeAnimTrailBoneIndecies()
   &FN_InitializeAnimTrail(AnimTrailWidth=1.1, Color=<rgba 0.21 0.21 0.21 0.1>, Lifetime=0.14, SkeletalMesh=$System.Skeletal Mesh) {module=FN_InitializeAnimTrail001}
   &AddVelocity(Velocity=<0 1000 0>, Velocity Speed Scale=2) {enabled=false}
  ]
  particleUpdateScript [
   &FN_UpdateAnimTrail(SkeletalMesh=$System.Skeletal Mesh) {module=FN_UpdateAnimTrail001}
   &SolveForcesAndVelocity(Clamp Velocity=false, Limit Acceleration=false)
   &DynamicMaterialParameters(
    Index 0 Param 1=$Emitter.RandomFloat
    Index 0 Param 4=&Multiply_Float(
     A=$System.SwingGeneralFade
     B=&FloatFromCurve(FloatCurve=DataInterface() {propertyValues {Curve {keys [
          {interpMode=RCIM_Cubic, tangentMode=RCTM_User, time=0.1, value=1}
          {arriveTangent=-0.0076, interpMode=RCIM_Cubic, leaveTangent=-0.0076, tangentMode=RCTM_User, time=0.8, value=0}
         ]}}})
    )
    Module.Param0WriteEnabled=true
    Module.Param1WriteEnabled=false
    Module.Param2WriteEnabled=false
    Module.Param3WriteEnabled=true
   ) {enabled=false, module=DynamicMaterialParameters001}
   &ScaleColor(
    Scale Alpha=&Multiply_Float(
     A=$System.SwingGeneralFade
     B=&FloatFromCurve(FloatCurve=DataInterface() {propertyValues {Curve {keys [
          {arriveTangent=-2.3015, interpMode=RCIM_Cubic, leaveTangent=-2.3015, tangentMode=RCTM_User, time=0, value=1}
          {interpMode=RCIM_Cubic, tangentMode=RCTM_User, time=1, value=0}
         ]}}})
    )
    Scale RGB=&MakeVectorFromLinearColorRGB(Color=&ColorFromCurve(ColorCurve=DataInterface() {propertyValues {
        AlphaCurve {keys [<at 0 1 RCIM_Cubic>]}
        BlueCurve {keys [<at 0 70 RCIM_Cubic>,<at 0.1692 10>,<at 0.8412 0.2>]}
        GreenCurve {keys [<at 0 49 RCIM_Cubic>,<at 0.1692 7.25>,<at 0.8412 0.02>]}
        RedCurve {keys [<at 0 38.5 RCIM_Cubic>,<at 0.1692 7>,<at 0.8412 0.02>]}
       }}))
    ScaleA=true
    ScaleRGB=true
   )
  ]
  rendererClasses [&NiagaraRibbonRendererProperties]
  renderers [{rendererClass=&NiagaraRibbonRendererProperties}]
 }
 SharpBladeTrail {
  emitterUpdateScript [
   &EmitterState()
   &FN_SpawnAnimTrail(AnimTrailWidthMode=::FromFirst, Spawn Interval=0)
  ]
  particleSpawnScript [
   &FN_InitializeAnimTrailBoneIndecies(First Bone or Socket Index=1, Second Bone or Socket Index=2)
   &FN_InitializeAnimTrail(AnimTrailWidth=0.46, Lifetime=0.3, SkeletalMesh=$System.Skeletal Mesh) {module=FN_InitializeAnimTrail001}
   &CameraOffset(Camera Offset Amount=2) {enabled=false}
   &AddVelocity(Velocity=<0 1000 0>) {enabled=false}
  ]
  particleUpdateScript [
   &FN_UpdateAnimTrail(Locked Particle RibbonWidth=1, SkeletalMesh=$System.Skeletal Mesh) {module=FN_UpdateAnimTrail001}
   &ScaleColor(
    Scale Alpha=&Multiply_Float(
     A=$System.SwingGeneralFade
     B=&FloatFromCurve(FloatCurve=DataInterface() {propertyValues {Curve {keys [
          {arriveTangent=-29.4069, interpMode=RCIM_Cubic, leaveTangent=-29.4069, tangentMode=RCTM_User, time=0.1, value=3}
          {arriveTangent=-0.939, interpMode=RCIM_Cubic, leaveTangent=-0.939, tangentMode=RCTM_User, time=0.25, value=0.4}
          <at 1 0 RCIM_Cubic>
         ]}}})
    )
    Scale RGB=&MakeVectorFromLinearColorRGB(Color=&ColorFromCurve(ColorCurve=DataInterface() {propertyValues {
        AlphaCurve {keys [<at 0 1 RCIM_Cubic>]}
        BlueCurve {keys [<at 0 1.6 RCIM_Cubic>,<at 0.0672 1 RCIM_Cubic>,<at 0.7262 0.1>]}
        GreenCurve {keys [<at 0 0.704 RCIM_Cubic>,<at 0.0672 0.06 RCIM_Cubic>,<at 0.7262 0.0062>]}
        RedCurve {keys [<at 0 0.064 RCIM_Cubic>,<at 0.0672 0.295 RCIM_Cubic>,<at 0.7262 0.0375>]}
       }}))
    ScaleA=true
    ScaleRGB=true
   )
   &ScaleColor(
    Scale Alpha=&Multiply_Float(
     A=$System.SwingGeneralFade
     B=&FloatFromCurve(FloatCurve=DataInterface() {propertyValues {Curve {keys [
          {arriveTangent=-12.8154, interpMode=RCIM_Cubic, leaveTangent=-12.8154, tangentMode=RCTM_User, time=0.2, value=3}
          {arriveTangent=-3.1164, interpMode=RCIM_Cubic, leaveTangent=-3.1164, tangentMode=RCTM_User, time=0.45, value=1}
          <at 1 0 RCIM_Cubic>
         ]}}})
    )
    Scale RGB=&MakeVectorFromLinearColorRGB(Color=&FNTimeOfDayColor(
      Night Color=<rgba 0.1925 0.2371 0.35 1>
      Morning Color=<rgba 0.8 0.5527 0.52 1>
      Day Color=<rgba 1 1 1 1>
      Evening Color=<rgba 0.55 0.3575 0.3735 1>
     ))
    ScaleA=true
    ScaleRGB=true
   ) {enabled=false, module=ScaleColor001}
   &SolveForcesAndVelocity(Clamp Velocity=false, Limit Acceleration=false)
   &ScaleRibbonWidth(Ribbon Width Scale=&FloatFromCurve(FloatCurve=DataInterface() {propertyValues {Curve {keys [{interpMode=RCIM_Cubic, tangentMode=RCTM_User, time=0, value=1}, {interpMode=RCIM_Cubic, tangentMode=RCTM_User, time=0.3, value=0.1}]}}}))
  ]
  rendererClasses [&NiagaraRibbonRendererProperties]
  renderers [{rendererClass=&NiagaraRibbonRendererProperties}]
 }
}
summary {
 systemName=NS_Pickaxe_VampireHunter_Galaxy_Trail
 emitters [
  {emitterName=GalaxyTrail, rendererClasses [&NiagaraRibbonRendererProperties]}
  {emitterName=NebulaTrail, rendererClasses [&NiagaraRibbonRendererProperties]}
  {emitterName=SharpBladeTrail, rendererClasses [&NiagaraRibbonRendererProperties]}
 ]
 userVariables [
  User.bIsFrontEnd {type=NiagaraBool}
  User.bIsFrontEndPreview {type=NiagaraBool}
  User.ParaGlideLeanAlpha {type=NiagaraFloat, defaultValue=0, description=""}
  User.RibbonIDOffset {type=NiagaraInt32}
 ]
}
